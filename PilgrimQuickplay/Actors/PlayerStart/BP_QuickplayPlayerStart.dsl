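; Quickplay player start that tracks its original team and owns a camera stage mark for its spawn point.
; On BeginPlay it saves the team and, if enabled, spawns the mark at its transform, tagged for removal between BattleStage rounds.
; GetAssociatedStageMark spawns the mark if missing.
; RegisterStageMark and UnregisterStageMark reach the Quickplay playspace's camera data provider, but their bodies are inverted: Register removes the mark, Unregister registers it.
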(event EventBeginPlay
  (Variables|Default|SetOriginalTeam (Math|Conversions|ToInteger(Byte) (Variables|Team|GetTeam)))
  (CallFunction|SpawnStageMark))

(event EventTick
  (params (DeltaSeconds :type Float%20%28single-precision%29)))

(fn GetAssociatedStageMark
  (Utilities|IsValid (Variables|Default|GetStageMark)
    (:Is%20Valid
      _)
    (:Is%20Not%20Valid
      (CallFunction|SpawnStageMark)))
  (return (Variables|Default|GetStageMark)))

(fn GetNewTeam
  (return (Math|Conversions|ToInteger(Byte) (Variables|Team|GetTeam))))

(fn RegisterStageMark
  (bind _returnvalue (Class|SparksMusicPlayspace|GetSparksMusicPlayspaceOfClass "/PilgrimQuickplay/Playspaces/BP_PilgrimQuickplayPlayspace.BP_PilgrimQuickplayPlayspace_C" 0 true))
  (bind _returnvalue_2 (Actor|GetComponentbyClass _returnvalue "/Script/SparksCameraDirectorRuntime.SparksCameraPluginDataProvider"))
  (Utilities|IsValid (Variables|Default|GetStageMark)
    (:Is%20Valid
      (Utilities|IsValid _returnvalue
        (:Is%20Valid
          (Utilities|IsValid _returnvalue_2
            (:Is%20Valid
              (Class|SparksCameraPluginDataProvider|RemoveStageMark _returnvalue_2 (Variables|Default|GetStageMark)))))))))

(fn SetNewTeam
  (params (InNewTeam :type Integer))
  (Variables|Team|SetTeam (Math|Conversions|ToByte(Integer) InNewTeam)))

(fn SpawnStageMark
  (if (Variables|Default|GetbShouldSpawnStageMark)
    (Utilities|IsValid (Variables|Default|GetStageMark)
      (:Is%20Not%20Valid
        (bind _returnvalue (Game|SpawnActorfromClass "/SparksCameraDirector/B_SparksCameraStageMark.B_SparksCameraStageMark_C" (Transformation|GetActorTransform) "AlwaysSpawn" "MultiplyWithRoot" self 0))
        (bind _output_get (Variables|Default|SetStageMark _returnvalue))
        ;[1 1 1] Mark this stage marker as being attached to a player spawn.  This helps us remove them in between BattleStage rounds.
        (GameplayTags|AddGameplayTag (Variables|/Script/SparksCameraDirectorRuntime.SparksCameraStageMark|GetCameraTags _output_get) "(TagName=\"Pilgrim.Quickplay.PlayerStart.Stage\")" :!class /Script/GameplayTags.BlueprintGameplayTagLibrary)))))

(fn UnregisterStageMark
  (bind _returnvalue (Class|SparksMusicPlayspace|GetSparksMusicPlayspaceOfClass "/PilgrimQuickplay/Playspaces/BP_PilgrimQuickplayPlayspace.BP_PilgrimQuickplayPlayspace_C" 0 true))
  (bind _returnvalue_2 (Actor|GetComponentbyClass _returnvalue "/Script/SparksCameraDirectorRuntime.SparksCameraPluginDataProvider"))
  (Utilities|IsValid (Variables|Default|GetStageMark)
    (:Is%20Valid
      (Utilities|IsValid _returnvalue
        (:Is%20Valid
          (Utilities|IsValid _returnvalue_2
            (:Is%20Valid
              (Class|SparksCameraPluginDataProvider|RegisterStageMark _returnvalue_2 (Variables|Default|GetStageMark)))))))))

(fn ConstructionScript)
