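; Player start actor for the local player in the Pilgrim Quickplay green room, with a pedestal that appears only while needed.
; When the start is claimed, it unhides itself and gives the pedestal query and physics collision.
; On BeginPlay it waits for the Pilgrim event router, then listens for the quickplay state began server event.
; When that state is SongGameplay, it hides the actor and disables pedestal collision.
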(event EventBeginPlay
  (bind _eventrouter (Pilgrim|PilgrimWaitForEventRouter 0)
    (:OnEventRouterAvailable
      ;[0.2899 0.9927 1 0.577] When the song gameplay begins, hide this actor and turn off collision for the pedestal
      (bind (_asynctaskproxy _payload) (GameplayEventRouter|StartListeningToEvent _eventrouter "/Script/PilgrimQPSharedRuntime.PilgrimEvent_QuickplayState_BeganServer")
        (:OnEventReceived
          (if (Utilities|Enum|Equal(Enum) (Utilities|Struct|BreakPilgrimEventQuickplayStateBase _payload) "SongGameplay")
            (Rendering|SetActorHiddenInGame true)
            (Collision|SetCollisionEnabled (Variables|Default|GetPedestal) "NoCollision")))))))

(event EventOnPlayerStartClaimed
  ;[0.2899 0.9927 1 0.577] When this start is claimed, make the actor visible and enable collision for the player
  (Rendering|SetActorHiddenInGame false)
  (Collision|SetCollisionEnabled (Variables|Default|GetPedestal) "QueryAndPhysics"))

(fn ConstructionScript)
